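# Translucent UI material for the drips on the Season 7 Neon Cat loading screen, with an animatable reveal.
# An exponential radial gradient centered on the top-middle edge passes through a narrow SmoothStep driven by the Reveal parameter.
# That mask multiplies a drip texture channel chosen by DripType through chained lerps.
# The first lerp reads red on both inputs, so only red and blue are used.
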Material /BRCosmetics/2dAssets/Loadingscreens/Season7/NeonCatLoading/M_UI_LoadingNeonCat_Drips_Mat
graph {
 properties {blendMode=BLEND_Translucent, materialDomain=MD_UI, pixelDepthOffsetMode=PDOM_Legacy, usedShadingModels=Default Lit}
 nodes [
  Constant3Vector() {constant=<rgba 1 1 1 0>}
  n1 = Constant2Vector() {r=0.5}
  n2 = Constant() {r=1.25}
  n3 = Constant() {r=1}
  n4 = ScalarParameter() {defaultValue=1, parameterName=Reveal, sliderMax=1}
  n5 = TextureSample() {texture=&/T_UI_LoadingNeonCat_Drips}
  n6 = ScalarParameter() {parameterName=DripType}
  Multiply(
   A=MaterialFunctionCall(
    Alpha=MaterialFunctionCall(CenterPosition=@n1, Radius=@n2, Density=@n3) {materialFunction=&RadialGradientExponential}.RadialGradientExponential
    Min=@n4
    Max=Add(A=@n4) {constB=0.051}
   ) {materialFunction=&SmoothStep}.Result
   B=LinearInterpolate(A=LinearInterpolate(A=@n5.R, B=@n5.R, Alpha=@n6), B=@n5.B, Alpha=Subtract(A=@n6))
  )
 ]
}
