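# Translucent UI material for the Season 7 Neon Cat loading screen image, revealing it through a spray-noise dissolve.
# Time since DecalSpawnTime, capped by Reveal and UnReveal, drives sharpened thresholds on the T_SprayTime noise.
# The Reveal mask and an inverted-noise UnReveal mask multiply with the texture's alpha.
# A third mask, capped slightly below Reveal, blends RevealColor into the texture color, tinting the reveal edge.
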Material /BRCosmetics/2dAssets/Loadingscreens/Season7/NeonCatLoading/M_UI_Loading_NeonCatImageMaster
graph {
 properties {blendMode=BLEND_Translucent, materialDomain=MD_UI, pixelDepthOffsetMode=PDOM_Legacy, usedShadingModels=Default Lit}
 nodes [
  n1 = TextureSampleParameter2D() {parameterName=Texture, texture=&/T_UI_Loading_Img1}
  n2 = TextureCoordinate() {uTiling=0.8, vTiling=0.7}
  n3 = ScalarParameter() {defaultValue=1.1, parameterName=Reveal, sliderMax=1.1}
  n4 = Subtract(A=Time() {bIgnorePause=true}, B=ScalarParameter() {parameterName=DecalSpawnTime})
  n5 = Constant() {desc=1 / Fade in duration, r=1.25}
  n6 = Min(A=Add(A=@n3) {constB=-0.1}, B=Saturate(None=Multiply(A=@n4, B=@n5)))
  n7 = Add(
   A=TextureSample(UVs=@n2) {texture=/Game/Global/Textures/Default/Noise/T_SprayTime}.R
   B=Add(A=Constant() {r=-1}, B=Multiply(A=@n6) {constB=1.1})
  )
  LinearInterpolate(
   A=VectorParameter() {parameterName=RevealColor, defaultValue=<rgba 1 0 0.0759 1>}.RGB
   B=@n1.RGB
   Alpha=Clamp(None=Multiply(A=@n7) {constB=10})
  )
  n8 = Constant() {r=-1}
  n9 = Multiply(
   A=Subtract(A=Time() {bIgnorePause=true}, B=ScalarParameter() {parameterName=DecalSpawnTime})
   B=Constant() {desc=1 / Fade in duration, r=1.25}
  )
  n10 = Multiply(A=Min(A=@n3, B=Saturate(None=@n9))) {constB=1.1}
  n11 = Multiply(A=Add(A=TextureSample(UVs=@n2) {texture=&T_SprayTime}.R, B=Add(A=@n8, B=@n10))) {constB=10}
  n12 = TextureSample(UVs=@n2) {texture=&T_SprayTime}
  n13 = Constant() {r=-1}
  n14 = Multiply(
   A=Subtract(A=Time() {bIgnorePause=true}, B=ScalarParameter() {parameterName=DecalSpawnTime})
   B=Constant() {desc=1 / Fade in duration, r=1.25}
  )
  n15 = Multiply(A=Min(A=ScalarParameter() {defaultValue=1.1, parameterName=UnReveal, sliderMax=1.1}, B=Saturate(None=@n14))) {constB=1.1}
  n16 = Multiply(A=Add(A=OneMinus(None=@n12.R), B=Add(A=@n13, B=@n15))) {constB=10}
  Multiply(A=@n1.A, B=Multiply(A=Clamp(None=@n11), B=Clamp(None=@n16)))
 ]
}
