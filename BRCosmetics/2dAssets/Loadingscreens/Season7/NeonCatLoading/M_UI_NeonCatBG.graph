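# UI material that draws the animated purple and magenta background for the Season 7 Neon Cat loading screen.
# A time-rotated UV gradient blends two base colors, then an exponential radial gradient mixes in magenta.
# A screen-aligned, rotated geocube texture supplies smoothstepped edges from green and a red-channel mask for the brighter of two panning noise layers.
# Both pattern terms are summed, clamped by CloudAmount and overlay-blended onto the color.
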Material /BRCosmetics/2dAssets/Loadingscreens/Season7/NeonCatLoading/M_UI_NeonCatBG
graph {
 properties {materialDomain=MD_UI, pixelDepthOffsetMode=PDOM_Legacy, usedShadingModels=Default Lit}
 nodes [
  n1 = Constant3Vector() {constant=<rgba 0.2159 0.0331 1 0>}
  n2 = Constant3Vector() {constant=<rgba 0.3278 0.0212 0.7758 0>}
  n3 = TextureCoordinate()
  n4 = Fmod(
   A=Multiply(A=Time() {period=1}, B=ScalarParameter() {defaultValue=1, parameterName=CornerColorRotationSpeed}) {constB=0.2}
   B=Constant() {r=1}
  )
  n5 = OneMinus(None=ComponentMask(None=MaterialFunctionCall(UVs=@n3, "Rotation Angle (0-1)"=@n4) {materialFunction=&CustomRotator}.Rotated Values) {r=true})
  n6 = VectorParameter() {group=Radial, parameterName=RadialPosition, defaultValue=<rgba 1.25 0 0 1>}
  n7 = AppendVector(A=@n6.R, B=@n6.R)
  n8 = ScalarParameter() {defaultValue=2, group=Radial, parameterName=RadialSize}
  n9 = ScalarParameter() {defaultValue=2, group=Radial, parameterName=RadialDensity}
  n10 = MaterialFunctionCall() {materialFunction=&ScreenAlignedUVs}
  n11 = VectorParameter() {group=Pattern, parameterName=PatternSize, defaultValue=<rgba 10 7 0 1>}
  n12 = AppendVector(A=@n11.R, B=@n11.R)
  n13 = TextureSample(UVs=MaterialFunctionCall(
    UVs=Multiply(A=@n10."X 100%, Y 100%", B=@n12)
    "Rotation Angle (0-1)"=ScalarParameter() {defaultValue=0.01, group=Pattern, parameterName=PatternRotation}
   ) {materialFunction=&CustomRotator}.Rotated Values) {texture=&/T_UI_Geocube_Pattern}
  n14 = ScalarParameter() {defaultValue=0.2, group=Pattern, parameterName=PatternEdge}
  n15 = ScalarParameter() {defaultValue=0.003, group=Pattern, parameterName=PatternSmoothness}
  n16 = Multiply(
   A=MaterialFunctionCall(Alpha=@n13.G, Min=@n14, Max=Add(A=@n14, B=@n15)) {materialFunction=&SmoothStep}.Result
   B=ScalarParameter() {defaultValue=0.2, group=Pattern, parameterName=PatternAmount}
  )
  n17 = Multiply(A=TextureCoordinate() {uTiling=0.1, vTiling=0.1}, B=ScalarParameter() {defaultValue=3, group=CloudPattern, parameterName=CloudPatternSize})
  n18 = VectorParameter() {group=CloudPattern, parameterName=CloudPatternSpeed, defaultValue=<rgba -0.05 -0.05 0 1>}
  n19 = AppendVector(A=@n18.R, B=@n18.R)
  n20 = TextureSampleParameter2D(UVs=Panner(Coordinate=@n17, Speed=@n19) {speedX=-0.05, speedY=-0.05}) {
   group=CloudPattern
   parameterName=NoisePattern
   samplerType=SAMPLERTYPE_LinearGrayscale
   texture=/Game/Global/Textures/Default/Noise/glitch_noise
  }
  n21 = ScalarParameter() {group=CloudPattern, parameterName=Pattern2Speed}
  n22 = TextureSampleParameter2D(UVs=Panner(Coordinate=@n17, Speed=Multiply(A=@n19, B=@n21)) {speedX=-0.05, speedY=-0.05}) {
   group=CloudPattern
   parameterName=NoisePattern2
   samplerType=SAMPLERTYPE_LinearGrayscale
   texture=&glitch_noise
  }
  n23 = Multiply(A=@n13.R, B=Max(A=@n20.R, B=@n22.R))
  n24 = VectorParameter() {group=CloudPattern, parameterName=CloudAmount, defaultValue=<rgba 0 0 0 1>}
  MaterialFunctionCall(
   Base=LinearInterpolate(
    A=LinearInterpolate(A=@n1, B=@n2, Alpha=@n5)
    B=Constant3Vector() {constant=<rgba 0.6308 0.0024 0.7157 0>}
    Alpha=MaterialFunctionCall(CenterPosition=@n7, Radius=@n8, Density=@n9) {materialFunction=&RadialGradientExponential}.RadialGradientExponential
   )
   Blend=Clamp(None=Add(A=@n16, B=@n23), Min=@n24.R, Max=@n24.R) {minDefault=0.5}
  ) {materialFunction=&Blend_Overlay}
  ScalarParameter() {parameterName=CornerColorAngle}
 ]
}
