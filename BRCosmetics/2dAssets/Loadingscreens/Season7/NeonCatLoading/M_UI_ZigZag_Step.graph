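# Translucent UI material in the Season 7 NeonCat loading screen folder that draws a procedural zigzag step shape with animatable reveals.
# SmoothSteps on individually rotatable UV axes form three bands, vertical-horizontal-vertical, trimmed against each other at the corners.
# Edge2, Edge4 and Edge6 act as band widths offset from Edge1, Edge3 and Edge5.
# Angle1–3Reveal parameters wipe each segment along its length before Max merges them into one mask.
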Material /BRCosmetics/2dAssets/Loadingscreens/Season7/NeonCatLoading/M_UI_ZigZag_Step
graph {
 properties {blendMode=BLEND_Translucent, materialDomain=MD_UI, pixelDepthOffsetMode=PDOM_Legacy, usedShadingModels=Default Lit}
 nodes [
  Constant3Vector() {constant=<rgba 1 1 1 0>}
  n1 = TextureCoordinate()
  n2 = ScalarParameter() {parameterName=Edge1Rotation, sliderMax=0.25, sliderMin=-0.25}
  n3 = ScalarParameter() {defaultValue=0.2, parameterName=Edge1}
  n4 = MaterialFunctionCall(
   Alpha=ComponentMask(None=MaterialFunctionCall(UVs=@n1, "Rotation Angle (0-1)"=@n2) {materialFunction=&CustomRotator}.Rotated Values) {r=true}
   Min=@n3
   Max=Add(A=@n3, B=ScalarParameter() {defaultValue=0.01, parameterName=EdgeSmoothness})
  ) {materialFunction=&SmoothStep}
  n5 = ComponentMask(None=MaterialFunctionCall(UVs=TextureCoordinate(), "Rotation Angle (0-1)"=ScalarParameter() {parameterName=Edge2Rotation, sliderMax=0.25, sliderMin=-0.25}) {materialFunction=&CustomRotator}.Rotated Values) {r=true}
  n6 = Add(A=@n3, B=ScalarParameter() {defaultValue=0.1, parameterName=Edge2, sliderMax=0.04})
  n7 = Add(A=@n6, B=ScalarParameter() {defaultValue=0.01, parameterName=EdgeSmoothness})
  n8 = TextureCoordinate()
  n9 = ScalarParameter() {parameterName=Edge4Rotation, sliderMax=0.25, sliderMin=-0.25}
  n10 = ScalarParameter() {defaultValue=0.5, parameterName=Edge3}
  n11 = Add(A=@n10, B=ScalarParameter() {defaultValue=0.1, parameterName=Edge4, sliderMax=0.04})
  n12 = MaterialFunctionCall(
   Alpha=ComponentMask(None=MaterialFunctionCall(UVs=@n8, "Rotation Angle (0-1)"=@n9) {materialFunction=&CustomRotator}.Rotated Values) {g=true}
   Min=@n11
   Max=Add(A=@n11, B=ScalarParameter() {defaultValue=0.01, parameterName=EdgeSmoothness})
  ) {materialFunction=&SmoothStep}
  n13 = Subtract(
   A=Subtract(
    A=@n4.Result
    B=MaterialFunctionCall(Alpha=@n5, Min=@n6, Max=@n7) {materialFunction=&SmoothStep}.Result
   )
   B=@n12.Result
  )
  n14 = TextureCoordinate()
  n15 = ScalarParameter() {parameterName=Angle1Reveal, sliderMax=1}
  n16 = MaterialFunctionCall(
   Alpha=OneMinus(None=ComponentMask(None=@n14) {g=true})
   Min=@n15
   Max=Add(A=@n15, B=ScalarParameter() {defaultValue=0.01, parameterName=EdgeSmoothness})
  ) {materialFunction=&SmoothStep}
  n17 = TextureCoordinate()
  n18 = ScalarParameter() {parameterName=Edge3Rotation, sliderMax=0.25, sliderMin=-0.25}
  n19 = MaterialFunctionCall(
   Alpha=ComponentMask(None=MaterialFunctionCall(UVs=@n17, "Rotation Angle (0-1)"=@n18) {materialFunction=&CustomRotator}.Rotated Values) {g=true}
   Min=@n10
   Max=Add(A=@n10, B=ScalarParameter() {defaultValue=0.01, parameterName=EdgeSmoothness})
  ) {materialFunction=&SmoothStep}
  n20 = TextureCoordinate()
  n21 = ScalarParameter() {parameterName=Edge6Rotation, sliderMax=0.25, sliderMin=-0.25}
  n22 = ScalarParameter() {defaultValue=0.7, parameterName=Edge5}
  n23 = Add(A=@n22, B=ScalarParameter() {defaultValue=0.1, parameterName=Edge6, sliderMax=0.04})
  n24 = MaterialFunctionCall(
   Alpha=ComponentMask(None=MaterialFunctionCall(UVs=@n20, "Rotation Angle (0-1)"=@n21) {materialFunction=&CustomRotator}.Rotated Values) {r=true}
   Min=@n23
   Max=Add(A=@n23, B=ScalarParameter() {defaultValue=0.01, parameterName=EdgeSmoothness})
  ) {materialFunction=&SmoothStep}
  n25 = Subtract(A=Subtract(A=Subtract(A=@n19.Result, B=@n12.Result), B=@n24.Result), B=OneMinus(None=@n4.Result))
  n26 = TextureCoordinate()
  n27 = ScalarParameter() {parameterName=Angle2Reveal, sliderMax=1}
  n28 = MaterialFunctionCall(
   Alpha=OneMinus(None=ComponentMask(None=@n26) {r=true})
   Min=@n27
   Max=Add(A=@n27, B=ScalarParameter() {defaultValue=0.01, parameterName=EdgeSmoothness})
  ) {materialFunction=&SmoothStep}
  n29 = OneMinus(None=ComponentMask(None=TextureCoordinate()) {g=true})
  n30 = ScalarParameter() {parameterName=Angle3Reveal, sliderMax=1}
  n31 = Add(A=@n30, B=ScalarParameter() {defaultValue=0.01, parameterName=EdgeSmoothness})
  n32 = MaterialFunctionCall(UVs=TextureCoordinate(), "Rotation Angle (0-1)"=ScalarParameter() {parameterName=Edge5Rotation, sliderMax=0.25, sliderMin=-0.25}) {materialFunction=&CustomRotator}
  n33 = ScalarParameter() {defaultValue=0.01, parameterName=EdgeSmoothness}
  n34 = Subtract(
   A=MaterialFunctionCall(Alpha=ComponentMask(None=@n32.Rotated Values) {r=true}, Min=@n22, Max=Add(A=@n22, B=@n33)) {materialFunction=&SmoothStep}.Result
   B=@n24.Result
  )
  n35 = OneMinus(None=@n19.Result)
  Max(
   A=Max(A=Multiply(A=@n13, B=@n16.Result), B=Multiply(A=@n25, B=@n28.Result))
   B=Multiply(
    A=MaterialFunctionCall(Alpha=@n29, Min=@n30, Max=@n31) {materialFunction=&SmoothStep}.Result
    B=Subtract(A=@n34, B=@n35)
   )
  )
 ]
}
